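# Sound cue for the overheat sound of the Jethro BoldTiger weapon.
# A mixer starts the overheat wave immediately and layers a louder overheat-complete wave on top after a fixed one-second delay.
# It uses weapon reload attenuation and the weapon foley sound class.
# A concurrency override allows one instance per owner and blocks new plays while it is active; debug is enabled on the cue.
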SoundCue /JethroWeaponGameplay/Assets/Audio/BoldTiger/Sounds/IP_SC_JethroBoldTiger_Overheat
root=Mixer_1
nodes [
 {id=Delay_0, children [WavePlayer_6], properties {delayMax=1, delayMin=1}}
 {id=Mixer_1, children [WavePlayer_5,Delay_0], properties {inputVolume [1,1.5]}}
 {id=WavePlayer_5, properties {soundWaveAssetPtr=&/IP_Jethro_BoldTiger_Overheat_01}}
 {id=WavePlayer_6, properties {soundWaveAssetPtr=&/IP_Jethro_BoldTiger_Overheat_Complete_01}}
]
properties {
 attenuationSettings=&Fort_Weapon_Reload_Att
 bDebug=true
 bOverrideConcurrency=true
 maxDistance=2250
 soundClassObject=&Fort_WeaponFoley
 volumeMultiplier=1
 concurrencyOverrides {bLimitToOwner=true, maxCount=1, resolutionRule=PreventNew}
}
